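; Rift Gauntlets dash knockback ability, extending a parent knockback ability, that launches the hit target and can be hotfixed to skip teammates.
; On launch it fires a gameplay cue and applies fall-damage immunity, with team and hotfix checks.
; Vehicle targets keep a scaled share of their velocity.
; Vehicles also receive mass-adjusted roll and yaw spin, signed and scaled by impact angle from a launcher-velocity trace, falling back to the effect's impact point.
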(event Ability|EventActivateAbility
  ;[0 0.15 0.03273 0.5] ActivateAbility: Knockback based on teammate status.
  (if (not (ScalableFloat|GetValueasBool (Variables|Hotfix|GetHotfix_AllowTeammateKnockback)))
    (switch Utilities|FlowControl|Switch|SwitchonEFortTeamAffiliation (Team|GetTeamAffiliation (Ability|GetAvatarActorfromActorInfo) (Ability|EffectContext|GetEffectCauser (Ability|GetGrantedbyEffectContext)))
      (:Friendly
        (Ability|EndAbility))
      (:Neutral
        _)
      (:Hostile
        _)))
  (|CallParentFunction))

(event EventDoLaunch
  (params (Vel :type Vector :default "0, 0, 0"))
  ;[0 0.15 0.000543 0.5] DoLaunch: Apply knockback gameplay cue.
  (|CallParentFunction Vel)
  (Ability|ExecuteGameplayCueOnOwner (Utilities|Struct|BreakGameplayCueTag (Variables|GameplayCue|GetKnockbackApplyGameplayCueTag)))
  (if (Team|OnSameTeam (Variables|KnockbackParent|GetTarget) (Variables|KnockbackParent|GetLauncher))
    _
    (elif (ScalableFloat|GetValueasBool (Variables|Hotfix|GetHotfix_ApplyFallDamageImmunityToNonTeammates))
      _))
  (Ability|ApplyGameplayEffectToOwner (Variables|Gameplay|GetFallDamageImmunityGameplayEffect)))

(fn CalcVehicleMult
  (params (InVec :type Vector :default "0, 0, 0"))
  ;[0.01004 0.174 0.1927] CalcVehicleMult: Include the velocity of the vehicle to the calculation.
  (Utilities|FlowControl|Sequence
    (:then_0
      (bind _asfort_vehicle_interface (Utilities|Casting|CastToFortVehicleInterface (Variables|KnockbackParent|GetTarget)))
      (bind _returnvalue (Vehicle|GetVehicleActor _asfort_vehicle_interface :!class /Script/FortniteGame.FortVehicleInterface))
      (Utilities|IsValid _returnvalue
        (:Is%20Valid
          (return (+ (* (Transformation|GetVelocity _returnvalue) (ScalableFloat|GetValueatLevel (Variables|Hotfix|GetHotfix_VehicleKnockbackVelocityMaintainMultiplier))) (|CallParentFunction InVec))))))
    (:then_1
      (return (|CallParentFunction InVec)))))

(fn LaunchVehicle
  (params (VehicleInterface :type Fort%20Vehicle%20Interface%20Interface) (LaunchVelocity :type Vector%20%28by%20ref%29 :default "0, 0, 0"))
  (bind (_bblockinghit _binitialoverlap _time _distance _location _impactpoint) (Collision|BreakHitResult (Ability|EffectContext|GetHitResult (Ability|GetGrantedbyEffectContext))))
  ;[0.03892 0.1925 0.2188] LaunchVehicle: Overridden to also add angular rotation to the vehicle based on the impact location.
  (|CallParentFunction VehicleInterface LaunchVelocity)
  (Utilities|FlowControl|Sequence
    (:then_0
      (bind _returnvalue (Vehicle|GetVehicleActor VehicleInterface :!class /Script/FortniteGame.FortVehicleInterface))
      (Utilities|IsValid _returnvalue
        (:Is%20Valid
          (bind _returnvalue_2 (Class|FortVehicleInterface|GetSkeletalMeshComponent(Message) _returnvalue))
          (bind (_hitlocation _hitnormal _bonename _outhit _returnvalue_3) (Collision|LineTraceComponent _returnvalue_2 (Transformation|GetActorLocation (Variables|KnockbackParent|GetLauncher)) (+ (Transformation|GetActorLocation (Variables|KnockbackParent|GetLauncher)) (* (Math|Vector|Normalize (Transformation|GetVelocity (Variables|KnockbackParent|GetLauncher))) 2e+02)) true false false))
          (bind _returnvalue_4 (FortPhysicsSimulationLibrary|GetMass _returnvalue))
          (Audio|BroadcastAbilitySpinThrustatLocation 0 _returnvalue (Math|Vector|MakeVector (Math|Float|SafeDivide (* (ScalableFloat|GetValueatLevel (Variables|Hotfix|GetHotfix_VehicleKnockbackRollMagnitude)) (Utilities|Operators|Multiply (Math|Float|Sign(Float) (.z (Math|Vector|CrossProduct (Math|Vector|Normalize LaunchVelocity) (Math|Vector|Normalize (- (select _returnvalue_3 _hitlocation _impactpoint) (Transformation|GetActorLocation _returnvalue)))))) (Math|Float|MapRangeClamped (Math|Float|ClampAngle (Math|Trig|Acos(Degrees) (Math|Vector|DotProduct (Math|Vector|Normalize (- (select _returnvalue_3 _hitlocation _impactpoint) (Transformation|GetActorLocation _returnvalue))) (Math|Vector|Normalize LaunchVelocity))) 0.0 9e+01) 0.0 9e+01 0.0 1.0) -1.0)) (Math|Float|SafeDivide _returnvalue_4 1e+02)) 0.0 (Math|Float|SafeDivide (* (ScalableFloat|GetValueatLevel (Variables|Hotfix|GetHotfix_VehicleKnockbackYawMagnitude)) (Utilities|Operators|Multiply (Math|Float|Sign(Float) (.z (Math|Vector|CrossProduct (Math|Vector|Normalize LaunchVelocity) (Math|Vector|Normalize (- (select _returnvalue_3 _hitlocation _impactpoint) (Transformation|GetActorLocation _returnvalue)))))) (Math|Float|MapRangeClamped (Math|Float|ClampAngle (Math|Trig|Acos(Degrees) (Math|Vector|DotProduct (Math|Vector|Normalize (- (select _returnvalue_3 _hitlocation _impactpoint) (Transformation|GetActorLocation _returnvalue))) (Math|Vector|Normalize LaunchVelocity))) 0.0 9e+01) 0.0 9e+01 0.0 1.0) -1.0)) (Math|Float|SafeDivide _returnvalue_4 1e+02))) 7.8e+02 "None"))))
    (:then_1
      (return))))
